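# Progress marker widget for a mid-match assigned objective: a diamond radial fill around a chest icon, with distance text in meters.
# Events set progress, distance and affiliation; activation plays a team intro for friendly objectives, otherwise an enemy warning intro.
# Completion plays an unlock animation, then fires CompleteAnimFinished.
# Construct always applies the friendly fill color, and curve-table min/max distance visibility helpers are defined but unused in the event graph.
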WidgetBlueprint /MidMatchAssignedGameplay/Items/Environmental/UI/WBP_MidMatchAssigned_Progress
animations [
 Anim_FadeIn {
  binding_count=2
  length_seconds=0.6667
  object=&self:Anim_FadeIn
  sequence {
   bindings [
    Text_distance {
     id=#1
     possessed_class=CommonTextBlock
     tracks [
      {
       type=MovieSceneFloatTrack
       display_name=Render Opacity
       sections [{type=MovieSceneFloatSection, end_frame=40, start_frame=0, channels [None {type=MovieSceneScriptingFloatChannel, keys [[0,0],[10,0],[40,1]]}]}]
      }
      {
       type=MovieScene2DTransformTrack
       sections [{
         type=MovieScene2DTransformSection
         end_frame=40
         start_frame=0
         channels [
          Translation.X {type=MovieSceneScriptingFloatChannel, const [0,1]}
          Translation.Y {type=MovieSceneScriptingFloatChannel, const [0,1]}
          Angle {type=MovieSceneScriptingFloatChannel, const [0,1]}
          Scale.X {type=MovieSceneScriptingFloatChannel, keys [[0,1],[10,1],[25,1.5],[40,1]]}
          Scale.Y {type=MovieSceneScriptingFloatChannel, keys [[0,1],[10,1],[25,1.5],[40,1]]}
          Shear.X {type=MovieSceneScriptingFloatChannel, const [0,1]}
          Shear.Y {type=MovieSceneScriptingFloatChannel, const [0,1]}
         ]
        }]
      }
     ]
    }
    Overlay_Main {
     id=#2
     possessed_class=Overlay
     tracks [
      {
       type=MovieSceneFloatTrack
       display_name=Render Opacity
       sections [{type=MovieSceneFloatSection, end_frame=30, start_frame=0, channels [None {type=MovieSceneScriptingFloatChannel, keys [[0,0],[30,1]]}]}]
      }
      {
       type=MovieScene2DTransformTrack
       sections [{
         type=MovieScene2DTransformSection
         end_frame=30
         start_frame=0
         channels [
          Translation.X {type=MovieSceneScriptingFloatChannel}
          Translation.Y {type=MovieSceneScriptingFloatChannel}
         ]
        }]
      }
     ]
    }
   ]
   display_rate {numerator=60}
  }
 }
 Anim_FadeOut {
  binding_count=2
  length_seconds=0.7667
  object=&self:Anim_FadeOut
  sequence {
   bindings [
    Text_distance {
     id=#3
     possessed_class=CommonTextBlock
     tracks [{
       type=MovieSceneFloatTrack
       display_name=Render Opacity
       sections [{type=MovieSceneFloatSection, end_frame=46, start_frame=0, channels [None {type=MovieSceneScriptingFloatChannel, keys [[0,1],[30,0]]}]}]
      }]
    }
    Overlay_Main {
     id=#2
     possessed_class=Overlay
     tracks [
      {
       type=MovieSceneFloatTrack
       display_name=Render Opacity
       sections [{type=MovieSceneFloatSection, end_frame=46, start_frame=0, channels [None {type=MovieSceneScriptingFloatChannel, keys [[0,1],[23,0]]}]}]
      }
      {
       type=MovieScene2DTransformTrack
       sections [{
         type=MovieScene2DTransformSection
         end_frame=46
         start_frame=0
         channels [
          Translation.X {type=MovieSceneScriptingFloatChannel}
          Translation.Y {type=MovieSceneScriptingFloatChannel}
         ]
        }]
      }
     ]
    }
   ]
   display_rate {numerator=60}
  }
 }
 Anim_Unlocked {
  binding_count=8
  length_seconds=2
  object=&self:Anim_Unlocked
  sequence {
   bindings [
    Image_Unlocked_Glow {
     id=#4
     possessed_class=Image
     tracks [
      {
       type=MovieSceneFloatTrack
       display_name=Render Opacity
       sections [{
         type=MovieSceneFloatSection
         end_frame=120
         start_frame=0
         channels [None {type=MovieSceneScriptingFloatChannel, keys [[0,0],[8,0],[26,1],[42,0],[72,1]]}]
        }]
      }
      {
       type=MovieSceneColorTrack
       display_name=Color and Opacity
       sections [{
         type=MovieSceneColorSection
         end_frame=120
         start_frame=0
         channels [
          Color.R {type=MovieSceneScriptingFloatChannel, const [0,2]}
          Color.G {type=MovieSceneScriptingFloatChannel, keys [[26,0.6769],[72,1]]}
          Color.B {type=MovieSceneScriptingFloatChannel, keys [[26,1],[72,0.4436]]}
          Color.A {type=MovieSceneScriptingFloatChannel, const [1,2]}
         ]
        }]
      }
     ]
    }
    Image_Unlocked_IconGlow_02 {
     id=#5
     possessed_class=Image
     tracks [
      {
       type=MovieSceneColorTrack
       display_name=Color and Opacity
       sections [{
         type=MovieSceneColorSection
         end_frame=120
         start_frame=0
         channels [
          Color.R {type=MovieSceneScriptingFloatChannel, const [0.6192,1]}
          Color.G {type=MovieSceneScriptingFloatChannel, const [1,1]}
          Color.B {type=MovieSceneScriptingFloatChannel, const [0,1]}
          Color.A {type=MovieSceneScriptingFloatChannel, thinned_from=5, keys [[26,0],[63,0],[91,0.3],[100,0]]}
         ]
        }]
      }
      {
       type=MovieScene2DTransformTrack
       sections [{
         type=MovieScene2DTransformSection
         end_frame=120
         start_frame=0
         channels [
          Angle {type=MovieSceneScriptingFloatChannel, const [0,1]}
          Scale.X {type=MovieSceneScriptingFloatChannel, keys [[53.999,1],[62.999,1.5],[72,0.8],[91,1]]}
          Scale.Y {type=MovieSceneScriptingFloatChannel, keys [[53.999,1],[62.999,1.5],[72,0.8],[91,1]]}
         ]
        }]
      }
     ]
    }
    Image_MMObj_Icon {
     id=#6
     possessed_class=Image
     tracks [
      {
       type=MovieSceneFloatTrack
       display_name=Render Opacity
       sections [{
         type=MovieSceneFloatSection
         end_frame=120
         start_frame=0
         channels [None {type=MovieSceneScriptingFloatChannel, keys [[0,1],[30,1],[39,0],[67,0]]}]
        }]
      }
      {
       type=MovieScene2DTransformTrack
       sections [{
         type=MovieScene2DTransformSection
         end_frame=120
         start_frame=0
         channels [
          Translation.X {type=MovieSceneScriptingFloatChannel, const [0,1]}
          Translation.Y {type=MovieSceneScriptingFloatChannel, const [0,1]}
          Angle {type=MovieSceneScriptingFloatChannel, const [0,1]}
          Scale.X {type=MovieSceneScriptingFloatChannel, keys [[0,1],[30,1.5]]}
          Scale.Y {type=MovieSceneScriptingFloatChannel, keys [[0,1],[30,1.5]]}
          Shear.X {type=MovieSceneScriptingFloatChannel, const [0,1]}
          Shear.Y {type=MovieSceneScriptingFloatChannel, const [0,1]}
         ]
        }]
      }
     ]
    }
    Image_Diamond_ProgressBar {
     id=#7
     possessed_class=Image
     tracks [{
       type=MovieSceneWidgetMaterialTrack
       sections [{
         type=MovieSceneParameterSection
         end_frame=120
         start_frame=0
         channels [
          ColorR {type=MovieSceneScriptingFloatChannel, const [0,2]}
          ColorG {type=MovieSceneScriptingFloatChannel, keys [[0,220],[21,220],[72,0.7231]]}
          ColorB {type=MovieSceneScriptingFloatChannel, keys [[0,255],[21,255],[72,0.0021]]}
          ColorA {type=MovieSceneScriptingFloatChannel, const [1,2]}
         ]
        }]
      }]
    }
    Image_Unlocked_IconGlow_01 {
     id=#8
     possessed_class=Image
     tracks [
      {
       type=MovieSceneColorTrack
       display_name=Color and Opacity
       sections [
        {
         type=MovieSceneColorSection
         end_frame=59
         start_frame=59
         channels [
          Color.R {type=MovieSceneScriptingFloatChannel, const [1,1]}
          Color.G {type=MovieSceneScriptingFloatChannel, const [1,1]}
          Color.B {type=MovieSceneScriptingFloatChannel, const [1,1]}
          Color.A {type=MovieSceneScriptingFloatChannel, const [1,1]}
         ]
        }
        {
         type=MovieSceneColorSection
         end_frame=120
         start_frame=0
         channels [
          Color.R {type=MovieSceneScriptingFloatChannel, keys [[0,1],[26,0]]}
          Color.G {type=MovieSceneScriptingFloatChannel, keys [[0,1],[26,0.8243]]}
          Color.B {type=MovieSceneScriptingFloatChannel, const [1,2]}
          Color.A {type=MovieSceneScriptingFloatChannel, keys [[0,0],[26,0.25],[39,0]]}
         ]
        }
       ]
      }
      {
       type=MovieScene2DTransformTrack
       sections [{
         type=MovieScene2DTransformSection
         end_frame=120
         start_frame=0
         channels [
          Translation.X {type=MovieSceneScriptingFloatChannel, const [0,1]}
          Translation.Y {type=MovieSceneScriptingFloatChannel, const [0,1]}
          Angle {type=MovieSceneScriptingFloatChannel, const [0,1]}
          Scale.X {type=MovieSceneScriptingFloatChannel, keys [[0,1],[26,1.2],[39,0]]}
          Scale.Y {type=MovieSceneScriptingFloatChannel, keys [[0,1],[26,1.2],[39,0]]}
          Shear.X {type=MovieSceneScriptingFloatChannel, const [0,1]}
          Shear.Y {type=MovieSceneScriptingFloatChannel, const [0,1]}
         ]
        }]
      }
     ]
    }
    Overlay_Main {
     id=#9
     possessed_class=Overlay
     tracks [{
       type=MovieSceneFloatTrack
       display_name=Render Opacity
       sections [{
         type=MovieSceneFloatSection
         end_frame=120
         start_frame=0
         channels [None {type=MovieSceneScriptingFloatChannel, keys [[0,1],[105,1],[120,0]]}]
        }]
      }]
    }
    Image_Unlocked_Icon {
     id=#10
     possessed_class=Image
     tracks [
      {
       type=MovieScene2DTransformTrack
       sections [{
         type=MovieScene2DTransformSection
         end_frame=120
         start_frame=0
         channels [
          Translation.X {type=MovieSceneScriptingFloatChannel}
          Translation.Y {type=MovieSceneScriptingFloatChannel}
          Angle {type=MovieSceneScriptingFloatChannel}
          Scale.X {type=MovieSceneScriptingFloatChannel, keys [[51,1.8],[62,1.8],[72,0.6],[91,1]]}
          Scale.Y {type=MovieSceneScriptingFloatChannel, keys [[51,1.5],[62,1.8],[72,0.6],[91,1]]}
          Shear.X {type=MovieSceneScriptingFloatChannel}
          Shear.Y {type=MovieSceneScriptingFloatChannel}
         ]
        }]
      }
      {
       type=MovieSceneFloatTrack
       display_name=Render Opacity
       sections [{
         type=MovieSceneFloatSection
         end_frame=120
         start_frame=0
         channels [None {type=MovieSceneScriptingFloatChannel, keys [[0,0],[51,0],[72,1],[115,1]]}]
        }]
      }
     ]
    }
    Overlay_UnlockedIcon {
     id=#11
     possessed_class=Overlay
     tracks [{
       type=MovieSceneFloatTrack
       display_name=Render Opacity
       sections [{
         type=MovieSceneFloatSection
         end_frame=120
         start_frame=0
         channels [None {type=MovieSceneScriptingFloatChannel, keys [[0,0],[8,0],[26,1],[45,1]]}]
        }]
      }]
    }
   ]
   display_rate {numerator=60}
  }
 }
 Anim_EnemyStart {
  binding_count=5
  length_seconds=2.2334
  object=&self:Anim_EnemyStart
  sequence {
   bindings [
    Text_distance {
     id=#12
     possessed_class=CommonTextBlock
     tracks [
      {
       type=MovieSceneFloatTrack
       display_name=Render Opacity
       sections [{
         type=MovieSceneFloatSection
         end_frame=120
         start_frame=0
         channels [None {type=MovieSceneScriptingFloatChannel, keys [[0,0],[70,0],[80,1]]}]
        }]
      }
      {
       type=MovieScene2DTransformTrack
       sections [{
         type=MovieScene2DTransformSection
         end_frame=134
         start_frame=0
         channels [
          Translation.X {type=MovieSceneScriptingFloatChannel}
          Translation.Y {type=MovieSceneScriptingFloatChannel}
          Angle {type=MovieSceneScriptingFloatChannel}
          Scale.X {type=MovieSceneScriptingFloatChannel, keys [[70,0.5],[100,1.5],[120,1]]}
          Scale.Y {type=MovieSceneScriptingFloatChannel, keys [[70,0.5],[100,1.5],[120,1]]}
          Shear.X {type=MovieSceneScriptingFloatChannel}
          Shear.Y {type=MovieSceneScriptingFloatChannel}
         ]
        }]
      }
     ]
    }
    Image_Warning_BorderOpaque {
     id=#13
     possessed_class=Image
     tracks [
      {
       type=MovieSceneByteTrack
       display_name=Visibility
       sections [{type=MovieSceneByteSection, end_frame=58, start_frame=57, channels [None {type=MovieSceneScriptingByteChannel}]}]
      }
      {
       type=MovieSceneColorTrack
       display_name=Color and Opacity
       sections [{
         type=MovieSceneColorSection
         end_frame=80
         start_frame=17
         channels [
          Color.R {type=MovieSceneScriptingFloatChannel}
          Color.G {type=MovieSceneScriptingFloatChannel}
          Color.B {type=MovieSceneScriptingFloatChannel}
          Color.A {type=MovieSceneScriptingFloatChannel, keys [[17.999,0],[50.999,0],[64.999,0.6],[79.999,0]]}
         ]
        }]
      }
      {
       type=MovieScene2DTransformTrack
       sections [{
         type=MovieScene2DTransformSection
         end_frame=80
         start_frame=47
         channels [
          Translation.X {type=MovieSceneScriptingFloatChannel}
          Translation.Y {type=MovieSceneScriptingFloatChannel}
          Angle {type=MovieSceneScriptingFloatChannel}
          Scale.X {type=MovieSceneScriptingFloatChannel, keys [[50.999,2.5],[52,2.4913],[79.999,0]]}
          Scale.Y {type=MovieSceneScriptingFloatChannel, keys [[50.999,2],[52,1.993],[79.999,0]]}
          Shear.X {type=MovieSceneScriptingFloatChannel, const [0,1]}
          Shear.Y {type=MovieSceneScriptingFloatChannel, const [0,1]}
         ]
        }]
      }
      {type=MovieSceneWidgetMaterialTrack, sections []}
      {
       type=MovieSceneFloatTrack
       display_name=Render Opacity
       sections [{type=MovieSceneFloatSection, end_frame=134, start_frame=96, channels [None {type=MovieSceneScriptingFloatChannel}]}]
      }
     ]
    }
    Overlay_WarningIcon {
     id=#14
     possessed_class=Overlay
     tracks [
      {
       type=MovieSceneFloatTrack
       display_name=Render Opacity
       sections [{
         type=MovieSceneFloatSection
         end_frame=134
         start_frame=0
         channels [None {type=MovieSceneScriptingFloatChannel, thinned_from=6, keys [[0,0],[41,1],[127,1],[134,0]]}]
        }]
      }
      {
       type=MovieScene2DTransformTrack
       sections [{
         type=MovieScene2DTransformSection
         end_frame=120
         start_frame=0
         channels [
          Translation.X {type=MovieSceneScriptingFloatChannel, const [0,4]}
          Translation.Y {type=MovieSceneScriptingFloatChannel, thinned_from=4, keys [[0,0],[70,0],[94,-90]]}
          Angle {type=MovieSceneScriptingFloatChannel, const [0,2]}
          Scale.X {type=MovieSceneScriptingFloatChannel, thinned_from=5, keys [[0,3],[33,1],[70,1],[94,0.7]]}
          Scale.Y {type=MovieSceneScriptingFloatChannel, thinned_from=5, keys [[0,3],[33,1],[70,1],[94,0.7]]}
          Shear.X {type=MovieSceneScriptingFloatChannel, const [0,2]}
          Shear.Y {type=MovieSceneScriptingFloatChannel, const [0,2]}
         ]
        }]
      }
     ]
    }
    Overlay_Main {
     id=#15
     possessed_class=Overlay
     tracks [{
       type=MovieSceneFloatTrack
       display_name=Render Opacity
       sections [{
         type=MovieSceneFloatSection
         end_frame=120
         start_frame=0
         channels [None {type=MovieSceneScriptingFloatChannel, keys [[0,0],[70,0],[80,1],[120,1]]}]
        }]
      }]
    }
    Overlay_Warning {
     id=#16
     possessed_class=Overlay
     tracks [{
       type=MovieSceneFloatTrack
       display_name=Render Opacity
       sections [{type=MovieSceneFloatSection, end_frame=134, start_frame=0, channels [None {type=MovieSceneScriptingFloatChannel, const [1,1]}]}]
      }]
    }
   ]
   display_rate {numerator=60}
  }
 }
 Anim_TeamStart {
  binding_count=2
  length_seconds=1.3333
  object=&self:Anim_TeamStart
  sequence {
   bindings [
    Overlay_Main {
     id=#17
     possessed_class=Overlay
     tracks [
      {
       type=MovieSceneFloatTrack
       display_name=Render Opacity
       sections [{type=MovieSceneFloatSection, end_frame=80, start_frame=0, channels [None {type=MovieSceneScriptingFloatChannel, keys [[0,0],[40,1]]}]}]
      }
      {
       type=MovieScene2DTransformTrack
       sections [{
         type=MovieScene2DTransformSection
         end_frame=80
         start_frame=0
         channels [
          Translation.X {type=MovieSceneScriptingFloatChannel, const [0,2]}
          Translation.Y {type=MovieSceneScriptingFloatChannel, const [0,2]}
          Angle {type=MovieSceneScriptingFloatChannel, const [0,1]}
          Scale.X {type=MovieSceneScriptingFloatChannel, keys [[0,3],[40,1],[50,1]]}
          Scale.Y {type=MovieSceneScriptingFloatChannel, keys [[0,3],[40,1],[50,1]]}
          Shear.X {type=MovieSceneScriptingFloatChannel, const [0,1]}
          Shear.Y {type=MovieSceneScriptingFloatChannel, const [0,1]}
         ]
        }]
      }
     ]
    }
    Text_distance {
     id=#18
     possessed_class=CommonTextBlock
     tracks [
      {
       type=MovieSceneFloatTrack
       display_name=Render Opacity
       sections [{type=MovieSceneFloatSection, end_frame=80, start_frame=0, channels [None {type=MovieSceneScriptingFloatChannel, keys [[0,0],[40,0],[50,1]]}]}]
      }
      {
       type=MovieScene2DTransformTrack
       sections [{
         type=MovieScene2DTransformSection
         end_frame=80
         start_frame=0
         channels [
          Translation.X {type=MovieSceneScriptingFloatChannel}
          Translation.Y {type=MovieSceneScriptingFloatChannel}
          Angle {type=MovieSceneScriptingFloatChannel}
          Scale.X {type=MovieSceneScriptingFloatChannel, keys [[30,0.5],[60,1.5],[80,1]]}
          Scale.Y {type=MovieSceneScriptingFloatChannel, keys [[30,0.5],[60,1.5],[80,1]]}
          Shear.X {type=MovieSceneScriptingFloatChannel}
          Shear.Y {type=MovieSceneScriptingFloatChannel}
         ]
        }]
      }
     ]
    }
   ]
   display_rate {numerator=60}
  }
 }
]
graphs [
 CacheAssignedLocalPlayerPawn {nodes [
   Inputs!() -> (In:Exec)
   > n1 = Get!(execute:Exec) -> (then:Exec, else:Exec, CachedLocalPawn:Pawn Object Reference)
   Outputs!(
    Valid:Exec=[@n8.then, @n1.then]
    CachedPawn:Pawn Object Reference=Get CachedLocalPawn() -> (CachedLocalPawn:Pawn Object Reference)
    Invalid:Exec=[@n7.Is Not Valid, @n5.Is Not Valid, @n3.Is Not Valid]
   )
   n2 = GetOwningPlayer()
   n3 = Is Valid(exec=@n1.else, InputObject=@n2)
   n4 = Get PlayerState!(self:Controller Object Reference=@n2) -> (PlayerState:Player State Object Reference)
   > n5 = Is Valid(InputObject=@n4)
   n6 = GetPawn!(self:Player State Object Reference=@n4) -> (ReturnValue:Pawn Object Reference)
   > n7 = Is Valid(InputObject=@n6)
   > n8 = Set CachedLocalPawn(CachedLocalPawn:Pawn Object Reference=@n6) -> (Output_Get:Pawn Object Reference)
  ]}
 CompleteAnimFinished {nodes [CompleteAnimFinished() {is=title K2Node_FunctionEntry}]}
 EventGraph {nodes [
   PlayCompleteAnimation() {is=title K2Node_CustomEvent}
   > n1 = "Play Animation with Finished event (legacy)"!(
    execute:Exec
    Widget:User Widget Object Reference=Self-Reference()
    InAnimation:Widget Animation Object Reference=Get Anim_Unlocked() -> (Anim_Unlocked:Widget Animation Object Reference)
    StartAtTime:float=0
    NumLoopsToPlay:int=1
    PlayMode:EUMGSequencePlayMode Enum
    PlaybackSpeed:float=1
   ) -> (then:Exec, Result:UMGSequence Player Object Reference, Finished:Exec) {is=title K2Node_PlayAnimation}
   Call Complete Anim Finished!(execute:Exec=@n1.Finished, self:WBP Mid Match Assigned Progress Object Reference) -> (then:Exec) {is=title K2Node_CallDelegate}
   OnObjectiveActivated() {is=title K2Node_CustomEvent}
   > PlayClaimingStartedAnims!(
    execute
    self
    ObjectiveAffiliation:EFortTeamAffiliation Enum=Get Objective Affiliation() -> (Objective Affiliation:EFortTeamAffiliation Enum=Friendly)
   ) -> (then)
   n2 = SetTextDistance!() -> (OutputDelegate, then, Display Distance:bool=~, Distance:double=~) {is=title K2Node_CustomEvent}
   > UpdateObjectiveDistance!(execute, self, DisplayDistance:bool=@n2.Display Distance, Distance:double=@n2.Distance) -> (then)
   n3 = SetObjectiveAffiliation!() -> (OutputDelegate, then, ObjectiveAffiliation:EFortTeamAffiliation Enum=Friendly) {is=title K2Node_CustomEvent}
   > Set Objective Affiliation(Objective Affiliation:EFortTeamAffiliation Enum=@n3.ObjectiveAffiliation) -> (Output_Get:EFortTeamAffiliation Enum=Friendly)
   n4 = UpdateProgress!() -> (OutputDelegate, then, ProgressPercent:double=~) {is=title K2Node_CustomEvent}
   > UpdateClaimProgress!(execute, self, ProgressPercent:double=@n4.ProgressPercent) -> (then)
   Event Construct()
   > n5 = GetDynamicMaterial(self=Get Image_Diamond_ProgressBar() -> (Image_Diamond_ProgressBar:Image Object Reference))
   > SetVectorParameterValue(self=@n5, ParameterName=FillColor, Value=Get FriendlyColor() -> (FriendlyColor:Linear Color Structure))
   PlayTeamStartAnimation()
   PlayEnemyStartAnimation()
   PlayFadeInAnimation()
   PlayFadeOutAnimation()
  ]}
 IsWidgetViewableAtPlayerDistance {nodes [
   n1 = IsWidgetViewableAtPlayerDistance!() -> (then, CurrentDistanceToPlayer:double=~) {is=title K2Node_FunctionEntry}
   n2 = Get SF_MinViewableDistance() -> (SF_MinViewableDistance:Scalable Float Structure)
   > n3 = Branch(Condition="float < float"(A=@n1, B=GetValueAtLevel!(Input:"Scalable Float Structure (by ref)"=@n2, Level:float=0) -> (ReturnValue:float)))
   > Return Node!(execute, IsViewableAtDistance:bool=false)
   n4 = Get SF_MaxViewableDistance() -> (SF_MaxViewableDistance:Scalable Float Structure)
   n5 = Branch(execute=@n3.else, Condition="float > float"(A=@n1, B=GetValueAtLevel(Input=@n4, Level=0)))
   > Return Node(IsViewableAtDistance=true)
   Return Node(execute=@n5.else, IsViewableAtDistance=Get bIsWidgetVisibleAtDistance() -> (bIsWidgetVisibleAtDistance:bool))
  ]}
 PlayClaimingStartedAnims {nodes [
   n1 = PlayClaimingStartedAnims!() -> (then, ObjectiveAffiliation:EFortTeamAffiliation Enum=Friendly) {is=title K2Node_FunctionEntry}
   n2 = Get Anim_EnemyStart() -> (Anim_EnemyStart:Widget Animation Object Reference)
   > PlayAnimation(InAnimation=Select!(
     Friendly:Widget Animation Object Reference=Get Anim_TeamStart() -> (Anim_TeamStart:Widget Animation Object Reference)
     Neutral:Widget Animation Object Reference=@n2
     Hostile:Widget Animation Object Reference=@n2
     Index:EFortTeamAffiliation Enum=@n1
    ) -> (ReturnValue:Widget Animation Object Reference))
  ]}
 PlayEnemyStartAnimation {nodes [
   PlayEnemyStartAnimation() {is=title K2Node_FunctionEntry}
   > PlayAnimation(InAnimation=Get Anim_EnemyStart())
  ]}
 PlayFadeInAnimation {nodes [
   PlayFadeInAnimation() {is=title K2Node_FunctionEntry}
   > PlayAnimation(InAnimation=Get Anim_FadeIn() -> (Anim_FadeIn:Widget Animation Object Reference))
  ]}
 PlayFadeOutAnimation {nodes [
   PlayFadeOutAnimation() {is=title K2Node_FunctionEntry}
   > PlayAnimation(InAnimation=Get Anim_FadeOut() -> (Anim_FadeOut:Widget Animation Object Reference))
  ]}
 PlayTeamStartAnimation {nodes [
   PlayTeamStartAnimation() {is=title K2Node_FunctionEntry}
   > PlayAnimation(InAnimation=Get Anim_TeamStart())
  ]}
 UpdateClaimProgress {nodes [
   n1 = UpdateClaimProgress!() -> (then, ProgressPercent:double=~) {is=title K2Node_FunctionEntry}
   > n2 = GetDynamicMaterial(self=Get Image_Diamond_ProgressBar())
   > SetScalarParameterValue(self=@n2, ParameterName=Progress, Value=@n1)
  ]}
 UpdateObjectiveDistance {nodes [
   n1 = UpdateObjectiveDistance!() -> (then, DisplayDistance:bool=~, Distance:double=~) {is=title K2Node_FunctionEntry}
   n2 = Floor(A=float / float(A=@n1.Distance, B=100)) {is=title K2Node_CallFunction}
   > Set Text(
    Text=Format Text!(Format="LOCTEXT(\"{Distance} m\")", Distance:Text="To Text (Integer)"(Value=@n2)) -> (Result)
    self=Get Text_distance() -> (Text_distance:Common Text Object Reference)
   )
   > SetVisibility(
    self=Get Text_distance()
    InVisibility=Select!(
     Option 0:ESlateVisibility Enum=Collapsed
     Option 1:ESlateVisibility Enum=HitTestInvisible
     Index:bool=Get DisplayDistance() -> (DisplayDistance:bool)
    ) -> (ReturnValue:ESlateVisibility Enum)
   )
  ]}
 ViewableDistanceFromPlayer {nodes [
   n1 = Inputs!() -> (In, DistanceFromOwningPlayer:double=~)
   n2 = GetValueAtLevel(Input=Get SF_MinViewableDistance(), Level=0)
   > n3 = Branch(Condition=AND Boolean(A="float < float"(A=@n1, B=@n2), B=Get bIsWidgetVisibleAtDistance()))
   > Outputs!(DeactivateWidget:Exec, ActivateWidget:Exec=@n6.then, HoldCurrent:Exec=@n6.else)
   n4 = Get bIsWidgetVisibleAtDistance()
   n5 = GetValueAtLevel(Input=Get SF_MaxViewableDistance(), Level=0)
   n6 = Branch(execute=@n3.else, Condition=AND Boolean(A="float > float"(A=@n1, B=@n5), B=NOT Boolean(A=@n4)))
  ]}
]
variables [
 bEvaluateIsWidgetVisible {type=boolean}
 bIsWidgetVisibleAtDistance {type=boolean}
 CachedLocalPawn {type=Pawn, default=null}
 CompleteAnimFinished {}
 FriendlyColor {type=LinearColor, default=<rgba 0.0273 0.5906 0.8632 1>}
 HostileColor {type=LinearColor, default=<rgba 0.8308 0.1651 0 1>}
 Objective Affiliation {type=EFortTeamAffiliation, default=Friendly}
 Objective Object {type=BGA_MidMatchAssignedObjective_Interaction_Parent_C, default=null}
 SF_MaxViewableDistance {
  type=ScalableFloat
  default {
   registryType=None
   value=1
   curve {
    curveTable=/MidMatchAssignedGameplay/DataTables/MidMatchAssignedObjectiveGameData
    rowName=MidMatchAssigned.UI.MaxDistanceToUnhideProgressWidgetInUU
   }
  }
 }
 SF_MinViewableDistance {
  type=ScalableFloat
  default {
   registryType=None
   value=1
   curve {
    curveTable=&MidMatchAssignedObjectiveGameData
    rowName=MidMatchAssigned.UI.MinDistanceToHideProgressWidgetInUU
   }
  }
 }
 SF_ObjectiveDuration {
  type=ScalableFloat
  default {
   registryType=None
   value=1
   curve {curveTable=&MidMatchAssignedObjectiveGameData, rowName=MidMatchAssigned.Interaction.ObjectiveCountdown}
  }
 }
 Timer_ObjectiveTime {type=TimerHandle}
 Timer_UpdateCircleTime {type=TimerHandle}
 TimerPollingUpdateInterval {type=number, default=0.2}
 TotalTimerDuration {type=number, default=-1}
]
widgets {
 info {inheritedWidgetCount=0, parentClass=&UserWidget, rootWidgetClass=&Overlay, widgetCount=17}
 widgets [
  Overlay_Root {widgetClassPath=&Overlay, widget=Overlay_Root}
  Overlay_Main {
   widgetClassPath=&Overlay
   parent=Overlay_Root
   widget=Overlay_Main
   properties {renderOpacity=0}
   slot {class=OverlaySlot, properties {horizontalAlignment=HAlign_Fill, verticalAlignment=VAlign_Fill}}
  }
  Image_Unlocked_Glow {
   widgetClassPath=&Image
   parent=Overlay_Main
   widget=Image_Unlocked_Glow
   properties {
    renderOpacity=0
    brush {
     bIsDynamicallyLoaded=false
     drawAs=Image
     imageType=NoImage
     mirroring=NoMirror
     resourceName=None
     resourceObject=/FortUILibrary/Styles/Border/Materials/M_Border_SoftGlow
     tiling=NoTile
     imageSize=<64 64>
     margin=<ltrb 0 0 0 0>
     outlineSettings {
      bUseBrushTransparency=false
      roundingType=HalfHeightRadius
      width=0
      color {colorUseRule=UseColor_Specified, specifiedColor=<rgba 0 0 0 0>}
      cornerRadii=<0 0 0 0>
     }
     tintColor {colorUseRule=UseColor_Specified, specifiedColor=<rgba 1 1 1 1>}
     uVRegion {bIsValid=false, max=<0 0>, min=<0 0>}
    }
    colorAndOpacity=<rgba 0.2472 1 0 1>
   }
   slot {class=OverlaySlot, properties {horizontalAlignment=HAlign_Fill, verticalAlignment=VAlign_Fill, padding=<ltrb -5 -5 -5 -5>}}
  }
  Overlay_ProgressBar {
   widgetClassPath=&Overlay
   parent=Overlay_Main
   widget=Overlay_ProgressBar
   slot {class=OverlaySlot, properties {horizontalAlignment=HAlign_Fill, verticalAlignment=VAlign_Fill}}
  }
  Image_Diamond_BGOpaque {
   widgetClassPath=&Image
   parent=Overlay_ProgressBar
   widget=Image_Diamond_BGOpaque
   properties {brush {
     bIsDynamicallyLoaded=false
     drawAs=Image
     imageType=NoImage
     mirroring=NoMirror
     resourceName=None
     resourceObject=/Game/Athena/HUD/Compass/M_UI_DiamondMarker
     tiling=NoTile
     imageSize=<32 32>
     margin=<ltrb 0 0 0 0>
     outlineSettings {
      bUseBrushTransparency=false
      roundingType=HalfHeightRadius
      width=0
      color {colorUseRule=UseColor_Specified, specifiedColor=<rgba 0 0 0 0>}
      cornerRadii=<0 0 0 0>
     }
     tintColor {colorUseRule=UseColor_Specified, specifiedColor=<rgba 0 0 0 0.5>}
     uVRegion {bIsValid=false, max=<0 0>, min=<0 0>}
    }}
   slot {class=OverlaySlot, properties {horizontalAlignment=HAlign_Fill, verticalAlignment=VAlign_Fill}}
  }
  Image_Diamond_ProgressBar {
   widgetClassPath=&Image
   bIsVariable=true
   parent=Overlay_ProgressBar
   widget=Image_Diamond_ProgressBar
   properties {
    visibility=SelfHitTestInvisible
    brush {
     bIsDynamicallyLoaded=false
     drawAs=Image
     imageType=NoImage
     mirroring=NoMirror
     resourceName=None
     resourceObject=/Game/Athena/HUD/Materials/M_UI_DiamondProgress_RadialFill_Inst
     tiling=NoTile
     imageSize=<68 68>
     margin=<ltrb 0 0 0 0>
     outlineSettings {
      bUseBrushTransparency=false
      roundingType=HalfHeightRadius
      width=0
      color {colorUseRule=UseColor_Specified, specifiedColor=<rgba 0 0 0 0>}
      cornerRadii=<0 0 0 0>
     }
     tintColor {colorUseRule=UseColor_Specified, specifiedColor=<rgba 1 1 1 1>}
     uVRegion {bIsValid=false, max=<0 0>, min=<0 0>}
    }
    renderTransform {angle=-45, scale=<1 1>, shear=<0 0>, translation=<0 0>}
   }
   slot {class=OverlaySlot, properties {horizontalAlignment=HAlign_Center, verticalAlignment=VAlign_Center}}
  }
  Image_Unlocked_IconGlow_01 {
   widgetClassPath=&Image
   parent=Overlay_Main
   widget=Image_Unlocked_IconGlow_01
   properties {
    brush {
     bIsDynamicallyLoaded=false
     drawAs=Image
     imageType=NoImage
     mirroring=NoMirror
     resourceName=None
     resourceObject=/MidMatchAssignedGameplay/Items/Environmental/Materials/MI_MMObj_Timer_Glow
     tiling=NoTile
     imageSize=<84 84>
     margin=<ltrb 0 0 0 0>
     outlineSettings {
      bUseBrushTransparency=false
      roundingType=HalfHeightRadius
      width=0
      color {colorUseRule=UseColor_Specified, specifiedColor=<rgba 0 0 0 0>}
      cornerRadii=<0 0 0 0>
     }
     tintColor {colorUseRule=UseColor_Specified, specifiedColor=<rgba 1 1 1 1>}
     uVRegion {bIsValid=false, max=<0 0>, min=<0 0>}
    }
    colorAndOpacity=<rgba 0 0.8243 1 0>
   }
   slot {class=OverlaySlot, properties {horizontalAlignment=HAlign_Center, verticalAlignment=VAlign_Center, padding=<ltrb -5 -5 -5 -5>}}
  }
  Image_MMObj_Icon {
   widgetClassPath=&Image
   bIsVariable=true
   parent=Overlay_Main
   widget=Image_MMObj_Icon
   properties {brush {
     bIsDynamicallyLoaded=false
     drawAs=Image
     imageType=NoImage
     mirroring=NoMirror
     resourceName=None
     resourceObject=/MidMatchAssignedGameplay/Items/Environmental/Textures/T_UI_Icon_MMObj_Timer_Chest
     tiling=NoTile
     imageSize=<52 52>
     margin=<ltrb 0 0 0 0>
     outlineSettings {
      bUseBrushTransparency=false
      roundingType=HalfHeightRadius
      width=0
      color {colorUseRule=UseColor_Specified, specifiedColor=<rgba 0 0 0 0>}
      cornerRadii=<0 0 0 0>
     }
     tintColor {colorUseRule=UseColor_Specified, specifiedColor=<rgba 1 1 1 1>}
     uVRegion {bIsValid=false, max=<0 0>, min=<0 0>}
    }}
   slot {class=OverlaySlot, properties {horizontalAlignment=HAlign_Center, verticalAlignment=VAlign_Center}}
  }
  Overlay_UnlockedIcon {
   widgetClassPath=&Overlay
   parent=Overlay_Main
   widget=Overlay_UnlockedIcon
   properties {renderOpacity=0}
   slot {class=OverlaySlot, properties {horizontalAlignment=HAlign_Center, verticalAlignment=VAlign_Center}}
  }
  Image_Unlocked_Icon {
   widgetClassPath=&Image
   parent=Overlay_UnlockedIcon
   widget=Image_Unlocked_Icon
   properties {brush {
     bIsDynamicallyLoaded=false
     drawAs=Image
     imageType=NoImage
     mirroring=NoMirror
     resourceName=None
     resourceObject=/Game/Athena/UI/InGame/Creative/PropertyEditor/Icon_Picker/Textures/128x128/T_UI_IconLibrary_unlocked_128
     tiling=NoTile
     imageSize=<56 56>
     margin=<ltrb 0 0 0 0>
     outlineSettings {
      bUseBrushTransparency=false
      roundingType=HalfHeightRadius
      width=0
      color {colorUseRule=UseColor_Specified, specifiedColor=<rgba 0 0 0 0>}
      cornerRadii=<0 0 0 0>
     }
     tintColor {colorUseRule=UseColor_Specified, specifiedColor=<rgba 1 1 1 1>}
     uVRegion {bIsValid=false, max=<0 0>, min=<0 0>}
    }}
   slot {class=OverlaySlot, properties {horizontalAlignment=HAlign_Center, verticalAlignment=VAlign_Center}}
  }
  Image_Unlocked_IconGlow_02 {
   widgetClassPath=&Image
   parent=Overlay_UnlockedIcon
   widget=Image_Unlocked_IconGlow_02
   properties {
    brush {
     bIsDynamicallyLoaded=false
     drawAs=Image
     imageType=NoImage
     mirroring=NoMirror
     resourceName=None
     resourceObject=&MI_MMObj_Timer_Glow
     tiling=NoTile
     imageSize=<84 84>
     margin=<ltrb 0 0 0 0>
     outlineSettings {
      bUseBrushTransparency=false
      roundingType=HalfHeightRadius
      width=0
      color {colorUseRule=UseColor_Specified, specifiedColor=<rgba 0 0 0 0>}
      cornerRadii=<0 0 0 0>
     }
     tintColor {colorUseRule=UseColor_Specified, specifiedColor=<rgba 1 1 1 1>}
     uVRegion {bIsValid=false, max=<0 0>, min=<0 0>}
    }
    colorAndOpacity=<rgba 0.6192 1 0 0>
   }
   slot {class=OverlaySlot, properties {horizontalAlignment=HAlign_Center, verticalAlignment=VAlign_Center, padding=<ltrb -5 -5 -5 -5>}}
  }
  Text_distance {
   widgetClassPath=&CommonTextBlock
   bIsVariable=true
   parent=Overlay_Main
   widget=Text_distance
   properties {
    justification=Center
    text=1000m
    visibility=HitTestInvisible
    font {
     bForceMonospaced=false
     bMaterialIsStencil=false
     fontMaterial=null
     fontObject=&BurbankSmall
     letterSpacing=0
     monospacedWidth=1
     size=16
     skewAmount=0
     typefaceFontName=Black
     outlineSettings {
      bApplyOutlineToDropShadows=false
      bMiteredCorners=false
      bSeparateFillAlpha=false
      outlineMaterial=null
      outlineSize=1
      outlineColor=<rgba 0 0 0 1>
     }
    }
    renderTransform {angle=0, scale=<1 1>, shear=<-12 0>, translation=<0 0>}
    shadowColorAndOpacity=<rgba 0.0052 0.0052 0.0056 1>
    shadowOffset=<4 4>
   }
   slot {class=OverlaySlot, properties {horizontalAlignment=HAlign_Center, verticalAlignment=VAlign_Bottom, padding=<ltrb 0 0 0 -30>}}
  }
  Overlay_Warning {
   widgetClassPath=&Overlay
   parent=Overlay_Root
   widget=Overlay_Warning
   properties {renderOpacity=0}
   slot {class=OverlaySlot, properties {horizontalAlignment=HAlign_Fill, verticalAlignment=VAlign_Fill}}
  }
  Image_Warning_BorderOpaque {
   widgetClassPath=&Image
   parent=Overlay_Warning
   widget=Image_Warning_BorderOpaque
   properties {
    brush {
     bIsDynamicallyLoaded=false
     drawAs=Image
     imageType=NoImage
     mirroring=NoMirror
     resourceName=None
     resourceObject=&M_UI_DiamondMarker
     tiling=NoTile
     imageSize=<32 32>
     margin=<ltrb 0 0 0 0>
     outlineSettings {
      bUseBrushTransparency=false
      roundingType=HalfHeightRadius
      width=0
      color {colorUseRule=UseColor_Specified, specifiedColor=<rgba 0 0 0 0>}
      cornerRadii=<0 0 0 0>
     }
     tintColor {colorUseRule=UseColor_Specified, specifiedColor=<rgba 1 0.0056 0 0.75>}
     uVRegion {bIsValid=false, max=<0 0>, min=<0 0>}
    }
    colorAndOpacity=<rgba 1 1 1 0>
   }
   slot {class=OverlaySlot, properties {horizontalAlignment=HAlign_Fill, verticalAlignment=VAlign_Fill, padding=<ltrb -15 -15 -15 -15>}}
  }
  Overlay_WarningIcon {
   widgetClassPath=&Overlay
   parent=Overlay_Warning
   widget=Overlay_WarningIcon
   slot {class=OverlaySlot, properties {horizontalAlignment=HAlign_Fill, verticalAlignment=VAlign_Fill}}
  }
  Image_Warning_BGOpaque {
   widgetClassPath=&Image
   parent=Overlay_WarningIcon
   widget=Image_Warning_BGOpaque
   properties {brush {
     bIsDynamicallyLoaded=false
     drawAs=Image
     imageType=NoImage
     mirroring=NoMirror
     resourceName=None
     resourceObject=&M_UI_DiamondMarker
     tiling=NoTile
     imageSize=<114 114>
     margin=<ltrb 0 0 0 0>
     outlineSettings {
      bUseBrushTransparency=false
      roundingType=HalfHeightRadius
      width=0
      color {colorUseRule=UseColor_Specified, specifiedColor=<rgba 0 0 0 0>}
      cornerRadii=<0 0 0 0>
     }
     tintColor {colorUseRule=UseColor_Specified, specifiedColor=<rgba 0.8475 0 0.012 1>}
     uVRegion {bIsValid=false, max=<0 0>, min=<0 0>}
    }}
   slot {class=OverlaySlot, properties {horizontalAlignment=HAlign_Center, verticalAlignment=VAlign_Center}}
  }
  Image_Warning_Icon {
   widgetClassPath=&Image
   parent=Overlay_WarningIcon
   widget=Image_Warning_Icon
   properties {brush {
     bIsDynamicallyLoaded=false
     drawAs=Image
     imageType=NoImage
     mirroring=NoMirror
     resourceName=None
     resourceObject=/Game/Athena/UI/InGame/Creative/PropertyEditor/Icon_Picker/Textures_V2/128x128/T_UI_IconLibrary_Caution_128
     tiling=NoTile
     imageSize=<52 52>
     margin=<ltrb 0 0 0 0>
     outlineSettings {
      bUseBrushTransparency=false
      roundingType=HalfHeightRadius
      width=0
      color {colorUseRule=UseColor_Specified, specifiedColor=<rgba 0 0 0 0>}
      cornerRadii=<0 0 0 0>
     }
     tintColor {colorUseRule=UseColor_Specified, specifiedColor=<rgba 1 1 1 1>}
     uVRegion {bIsValid=false, max=<0 0>, min=<0 0>}
    }}
   slot {class=OverlaySlot, properties {horizontalAlignment=HAlign_Center, verticalAlignment=VAlign_Center}}
  }
 ]
}
